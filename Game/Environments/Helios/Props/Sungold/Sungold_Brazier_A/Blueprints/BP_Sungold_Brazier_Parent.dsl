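; Parent Blueprint for the Sungold brazier prop in the Helios environment, controlling whether its static fire is visible.
; A Disable/EnableFlames function hides the Fire_Static component when the DisableFlames flag is set and shows it otherwise, without propagating to child components.
; The function runs in the construction script, after the parent's, and again on BeginPlay.
; Overlap and Tick events are empty, so the flag is not re-checked during play.
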(event EventBeginPlay
  (CallFunction|Disable/EnableFlames))

(event Collision|EventActorBeginOverlap
  (params (OtherActor :type Actor%20Object%20Reference)))

(event EventTick
  (params (DeltaSeconds :type Float%20%28single-precision%29)))

(fn Disable/EnableFlames
  (if (Variables|Building|TimeOfDayLights|GetDisableFlames)
    (Rendering|SetVisibility (Variables|Default|GetFire_Static) false false)
    (else
      (Rendering|SetVisibility (Variables|Default|GetFire_Static) true false))))

(fn ConstructionScript
  (|CallParentFunction)
  (CallFunction|Disable/EnableFlames))
